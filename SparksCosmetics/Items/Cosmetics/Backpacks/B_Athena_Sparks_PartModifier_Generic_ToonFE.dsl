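; Generic part modifier Blueprint for Sparks cosmetic backpacks, applying toon frontend LUT replacement, as its name and calls indicate.
; On the post-res-in event it runs the parent handler, then calls the ToonFrontendLUTReplace utility.
; The construction script only calls its parent, so all other behavior is inherited.
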(event EventOnPostResIn
  (|CallParentFunction)
  (Utilities|ToonFrontendLUTReplace))

(fn ConstructionScript
  (|CallParentFunction))
